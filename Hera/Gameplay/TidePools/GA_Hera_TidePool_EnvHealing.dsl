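; Gameplay ability for Hera tide pools that applies environmental healing to the player.
; On activation it runs the parent healing logic only if the healing-enabled scalable float is true, otherwise it ends immediately.
; Each heal tick runs the parent heal only when the avatar is a valid player pawn that is not inside any storm.
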(event Ability|EventActivateAbility
  (if (ScalableFloat|GetValueasBool (Variables|Default|GetHF_HealingEnabled))
    (|CallParentFunction)
    (else
      (Ability|EndAbility))))

(event EventHeal
  (bind _outfortplayerpawn (CallFunction|GetAvatarasValidFortPlayerPawn)
    (:ValidFortPlayerPawn
      (if (not (Variables|/Script/FortniteGame.FortPlayerPawn|GetbIsInAnyStorm _outfortplayerpawn))
        (|CallParentFunction)))))
